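# Physics asset for the Snowman Fashion pickaxe, defining collision bodies and joints for the fur bones on the axe mesh.
# Eight box bodies on the numbered dyn_fur bones each connect directly to a capsule on dyn_fur_mstr, forming a hub layout rather than a chain.
# All eight constraints share the same angular limits, allowing wider swing on one axis than the other and very little twist.
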PhysicsAsset /BRCosmetics/Weapons/FORT_Melee/Pickaxe_Snowman_Fashion/Meshes/Snowman_Fashion_Axe_Physics
bodies {
 dyn_fur_01 [{shapeType=Box, extentX=3.6656, extentY=7.6668, extentZ=3.8161, center=<3.1041 0.0127 0.2084>}]
 dyn_fur_02 [{shapeType=Box, extentX=3.5249, extentY=7.2579, extentZ=4.1616, center=<3.4448 0.0052 0.3167>}]
 dyn_fur_03 [{shapeType=Box, extentX=3.7926, extentY=6.559, extentZ=2.8071, center=<3.2068 0.2012 0.6726>}]
 dyn_fur_04 [{shapeType=Box, extentX=3.9625, extentY=9.5401, extentZ=3.4393, center=<3.2377 -0.0972 -0.2858>}]
 dyn_fur_05 [{shapeType=Box, extentX=3.4308, extentY=5.2758, extentZ=3.5356, center=<0.0587 0.0035 -2.4089>}]
 dyn_fur_06 [{shapeType=Box, extentX=3.2869, extentY=9.1138, extentZ=2.63, center=<2.2741 -0.5204 -0.2493>}]
 dyn_fur_07 [{shapeType=Box, extentX=2.4189, extentY=5.4177, extentZ=2.8204, center=<2.7565 -0.0552 -0.2663>}]
 dyn_fur_08 [{shapeType=Box, extentX=3.7212, extentY=6.5068, extentZ=2.7614, center=<2.2793 0.0025 -0.387>}]
 dyn_fur_mstr [{shapeType=Capsule, length=14.0112, radius=2.4387, center=<-5.124 0 0>, rotation=<rot 90 0 0>}]
}
constraints [
 {bone1Name=dyn_fur_01, bone2Name=dyn_fur_mstr, swing1LimitDegrees=15, swing2LimitDegrees=45, twistLimitDegrees=3}
 {bone1Name=dyn_fur_02, bone2Name=dyn_fur_mstr, swing1LimitDegrees=15, swing2LimitDegrees=45, twistLimitDegrees=3}
 {bone1Name=dyn_fur_03, bone2Name=dyn_fur_mstr, swing1LimitDegrees=15, swing2LimitDegrees=45, twistLimitDegrees=3}
 {bone1Name=dyn_fur_04, bone2Name=dyn_fur_mstr, swing1LimitDegrees=15, swing2LimitDegrees=45, twistLimitDegrees=3}
 {bone1Name=dyn_fur_05, bone2Name=dyn_fur_mstr, swing1LimitDegrees=15, swing2LimitDegrees=45, twistLimitDegrees=3}
 {bone1Name=dyn_fur_06, bone2Name=dyn_fur_mstr, swing1LimitDegrees=15, swing2LimitDegrees=45, twistLimitDegrees=3}
 {bone1Name=dyn_fur_07, bone2Name=dyn_fur_mstr, swing1LimitDegrees=15, swing2LimitDegrees=45, twistLimitDegrees=3}
 {bone1Name=dyn_fur_08, bone2Name=dyn_fur_mstr, swing1LimitDegrees=15, swing2LimitDegrees=45, twistLimitDegrees=3}
]
